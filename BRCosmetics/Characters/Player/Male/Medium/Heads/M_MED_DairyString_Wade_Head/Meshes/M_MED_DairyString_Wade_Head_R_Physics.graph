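# Physics asset for the DairyString Wade male medium head cosmetic, simulating its left and right dangling flaps.
# Small capsule bodies sit at each flap base, with box bodies covering the flaps: one segment on the right, a two-segment chain on the left.
# Each segment is constrained to its parent with asymmetric swing limits and narrow twist, tighter on the right, so flaps bend mainly along one axis.
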PhysicsAsset /BRCosmetics/Characters/Player/Male/Medium/Heads/M_MED_DairyString_Wade_Head/Meshes/M_MED_DairyString_Wade_Head_R_Physics
bodies {
 dyn_flapBase_r [{shapeType=Capsule, length=0.505, radius=0.505, shapeName=dyn_flapBase_r_capsule, rotation=<rot 0 -90 0>}]
 dyn_flap_01_l [{shapeType=Box, extentX=4.6459, extentY=8.7066, extentZ=15.9202, center=<3.4123 0.0003 0>, rotation=<rot 0 90 0>}]
 dyn_flap_02_l [{shapeType=Box, extentX=4.6552, extentY=8.7301, extentZ=9.2716, center=<4.1931 0 0>, rotation=<rot 0 90 0>}]
 dyn_flap_Base_l [{shapeType=Capsule, length=0.505, radius=0.505, shapeName=dyn_flap_Base_l_capsule, rotation=<rot 0 -90 0>}]
 dyn_flap_r [{shapeType=Box, extentX=5.052, extentY=16.1483, extentZ=12.6474, center=<7.1406 -0.0001 0>, rotation=<rot 0 90 0>}]
}
constraints [
 {bone1Name=dyn_flap_r, bone2Name=dyn_flapBase_r, swing1LimitDegrees=45, swing2LimitDegrees=10, twistLimitDegrees=3}
 {bone1Name=dyn_flap_01_l, bone2Name=dyn_flap_Base_l, swing1LimitDegrees=45, swing2LimitDegrees=10, twistLimitDegrees=12}
 {bone1Name=dyn_flap_02_l, bone2Name=dyn_flap_01_l, swing1LimitDegrees=45, swing2LimitDegrees=10, twistLimitDegrees=12}
]
