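# Animation Blueprint for the Puck skeleton of the DairyString Wade male head cosmetic, adding rig and physics motion to the incoming pose.
# A Control Rig feeds two chained RigidBody simulations in head-bone space, using separate left and right physics assets.
# ADS alphas weight the right-side simulation and an interpolated yaw override on the right flap base bone.
# The update event has no logic attached.
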AnimBlueprint /BRCosmetics/Characters/Player/Male/Medium/Heads/M_MED_DairyString_Wade_Head/Meshes/M_MED_DairyString_Wade_Head_Puck_AnimBP
target_skeleton=&/M_MED_DairyString_Wade_Head_Puck_Skeleton
graphs [
 AnimGraph {nodes [
   n1 = Control Rig(Source=Input Pose()) {
    alphaBoolBlend {blendOption=Linear}
    controlRigAssetReference {blueprintRigClass=&/M_MED_DairyString_Wade_Head_Puck_CtrlRig.M_MED_DairyString_Wade_Head_Puck_CtrlRig_C}
   }
   n2 = RigidBody!(ComponentPose=Local To Component(LocalPose=@n1), Alpha, ComponentLinearAccScale:Vector="0,0,0", ComponentLinearVelScale:Vector="0,0,0") -> (Pose) {
    bClampLinearTranslationLimitToRefPose=true
    bFreezeIncomingPoseOnStart=true
    bOverrideWorldGravity=true
    overridePhysicsAsset=&/M_MED_DairyString_Wade_Head_L_Physics
    simulationSpace=BaseBoneSpace
    alphaBoolBlend {blendOption=Linear}
    baseBoneRef {boneName=head}
    componentAppliedLinearAccClamp=<100 100 100>
    componentLinearAccScale=<0.6 0.2 0.2>
    componentLinearVelScale=<0.6 0.2 0.2>
    overrideWorldGravity {z=-980}
    simSpaceSettings {
     dampingAlpha=0
     maxAngularAcceleration=40
     maxAngularVelocity=8
     maxLinearAcceleration=1500
     maxLinearVelocity=300
     worldAlpha=1
     externalLinearDragV=<1 1 1>
    }
   }
   n3 = RigidBody(
    ComponentPose=@n2
    Alpha=Get ADSRigidBodyAlpha() -> (ADSRigidBodyAlpha:float)
    ComponentLinearAccScale="0,0,0"
    ComponentLinearVelScale="0,0,0"
   ) {
    bClampLinearTranslationLimitToRefPose=true
    bFreezeIncomingPoseOnStart=true
    bOverrideWorldGravity=true
    overridePhysicsAsset=&/M_MED_DairyString_Wade_Head_R_Physics
    simulationSpace=BaseBoneSpace
    alphaBoolBlend {blendOption=Linear}
    baseBoneRef {boneName=head}
    componentAppliedLinearAccClamp=<100 100 100>
    componentLinearAccScale=<0.6 0.2 0.2>
    componentLinearVelScale=<0.6 0.2 0.2>
    overrideWorldGravity {z=-980}
    simSpaceSettings {
     dampingAlpha=0
     maxAngularAcceleration=40
     maxAngularVelocity=8
     maxLinearAcceleration=1500
     maxLinearVelocity=300
     worldAlpha=1
     externalLinearDragV=<1 1 1>
    }
   }
   n4 = "Transform (Modify) Bone - Bone: dyn_flapBase_r"(ComponentPose=@n3, Alpha=Get ADSFlapAlpha() -> (ADSFlapAlpha:float), Translation="0,0,0", Rotation="0,-20,0", Scale="1,1,1") {
    rotationMode=BMM_Replace
    rotationSpace=BCS_BoneSpace
    alphaBoolBlend {blendOption=Linear}
    alphaScaleBiasClamp {bInterpResult=true, interpSpeedDecreasing=3.5, interpSpeedIncreasing=8}
    boneToModify {boneName=dyn_flapBase_r}
    rotation {yaw=-20}
   }
   Output Pose(Result=Component To Local(ComponentPose=@n4)) {layerGroup=None}
  ]}
 EventGraph {nodes [Event BlueprintUpdateAnimation()]}
]
variables []
